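# M_ICM_Wireframe is an unlit wireframe material in the InstancedMeshesCluster tech-art tool folder.
# It renders two-sided with the wireframe flag on, drawing mesh edges instead of filled surfaces.
# The graph holds only one constant light-blue color node, with no shown link to a material output, and has no parameters or textures.
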Material /Game/ArtTools/Tech_Art/InstancedMeshesCluster/Assets/M_ICM_Wireframe
graph {
 properties {pixelDepthOffsetMode=PDOM_Legacy, shadingModel=MSM_Unlit, twoSided=true, usedShadingModels=Unlit, wireframe=true}
 nodes [Constant3Vector() {constant=<rgba 0.5 0.5 1 1>}]
}
